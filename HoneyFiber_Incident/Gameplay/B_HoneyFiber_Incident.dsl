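; Weapon Blueprint for the HoneyFiber Incident throwable item, handling use audio, projectile spawning and throw effects.
; On cosmetic-capable instances, trigger presses check instigator health, playing initial-use audio below 100 and otherwise unbinding the handler.
; The throw ability calls a server event that spawns the projectile and sends a ThrowFX gameplay event, while a multicast event spawns the Niagara trail.
; Unequipping with no ammo left removes the item from inventory.
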(event EventBeginPlay
  (|CallParentFunction)
  (Networking|CanExecuteCosmeticEvents
    (:True
      (EventDispatchers|BindEventtoOnPressTrigger (delegate Custom|TryUse)))))

(event Custom|TryUse
  (params (Weapon :type Fort%20Weapon%20Object%20Reference))
  (bind _asfort_player_pawn (Utilities|Casting|CastToFortPlayerPawn (Game|GetInstigator)))
  (if (not (>= (FortPawn|GetHealth _asfort_player_pawn) 1e+02))
    (Class|BPInitialUseSoundComponent|TriggerFirstAudio (Variables|Default|GetBP_InitialUseSoundComponent))
    (EventDispatchers|UnbindEventfromOnPressTrigger (delegate Custom|TryUse))))

(event FortWeapon|EventOnUnEquippedAndDetached
  ;[0.15 0.04953 0.1208 0.5] Handle removal of the item if the player switches out just as ammo reaches zero
  (if (<= (FortWeapon|GetRemainingAmmo) 0)
    (Inventory|RemoveItemFromPlayerByGuid (CallFunction|GetPlayerController) (FortWeapon|GetInventoryGUID) 1 true)))

(event Custom|Server_SpawnProjectile
  (params (SpawnLocation :type Vector) (SpawnRotation :type Rotator))
  (bind _returnvalue (Game|GetInstigator))
  ;[0.01624 0.15 0.067 0.5] Spawn projectile (called from projectile throw ability)
  (bind _returnvalue_2 (Game|SpawnActorfromClass "/HoneyFiber_Incident/Gameplay/B_Prj_HoneyFiber_Incident.B_Prj_HoneyFiber_Incident_C" (Math|Transform|MakeTransform SpawnLocation SpawnRotation) "Undefined" "MultiplyWithRoot" _returnvalue 1.0 _returnvalue))
  (Ability|SendGameplayEventtoActor _returnvalue "(TagName=\"Event.HoneyFiber.Incident.ThrowFX\")" (Utilities|Struct|MakeGameplayEventData "(TagName=\"\")" _returnvalue_2)))

(event Custom|Multicast_PlayThrowFX
  (params (ProjectileRef :type Fort%20Projectile%20Base%20Object%20Reference))
  ;[0.01624 0.15 0.067 0.5] Display throw FX
  (Niagara|SpawnSystemAttached "/HoneyFiber_Incident/Assets/FX/NS_HoneyFiber_Incident_ProjectileTrail.NS_HoneyFiber_Incident_ProjectileTrail"))

(fn ConstructionScript
  (|CallParentFunction))
