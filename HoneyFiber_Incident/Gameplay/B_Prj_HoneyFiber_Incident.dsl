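; HoneyFiber Incident projectile that applies team-dependent impact effects, including a healing cue for damaged friendly players and an optional crumb trail on hostiles.
; ScalableFloats tune its launch velocity, gravity, delayed trail reveal and AoE mode.
; On stop it detaches visuals and delays destruction, applying effects to the hit actor or to async Targeting System AoE results.
; ApplyEffectsToTarget runs on authority, tracks already-hit actors and selects GameplayEffects and cues by affiliation.
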(event EventBeginPlay
  ; Begin play
  (Audio|Components|Audio|Play (Variables|Default|GetSound) 0.0)
  (Utilities|FlowControl|Delay (ScalableFloat|GetValueatLevel (Variables|Default|GetHFProjectileRevealDelay)))
  (Rendering|SetVisibility (Variables|Default|GetProjectileTrail) true true)
  (Variables|/Script/Engine.ProjectileMovementComponent|SetProjectileGravityScale (ScalableFloat|GetValueatLevel (Variables|Default|GetHFProjectileGravity)) (Variables|Movement|GetProjectileMovementComponent))
  (Projectile|SetVelocity (Variables|Movement|GetProjectileMovementComponent) (* (Transformation|GetForwardVector (Variables|Collision|GetCapsuleComponent)) (ScalableFloat|GetValueatLevel (Variables|Default|GetHFProjectileVelocity)))))

(event Collision|EventActorBeginOverlap
  (params (OtherActor :type Actor%20Object%20Reference))
  ;[1 1 1] Destroy on overlap with a target
  (if (!= OtherActor (Variables|Actor|GetInstigator))
    _))

(event EventOnStop
  (params (Hit_bBlockingHit :type Boolean) (Hit_bInitialOverlap :type Boolean) (Hit_Time :type Float%20%28single-precision%29) (Hit_Distance :type Float%20%28single-precision%29) (Hit_Location :type Vector :default "0, 0, 0") (Hit_ImpactPoint :type Vector :default "0, 0, 0") (Hit_Normal :type Vector :default "0, 0, 0") (Hit_ImpactNormal :type Vector :default "0, 0, 0") (Hit_PhysMat :type Physical%20Material%20Object%20Reference) (Hit_HitActor :type Actor%20Object%20Reference) (Hit_HitComponent :type Primitive%20Component%20Object%20Reference) (Hit_HitBoneName :type Name :default "None") (Hit_BoneName :type Name :default "None") (Hit_HitItem :type Integer) (Hit_ElementIndex :type Integer) (Hit_FaceIndex :type Integer) (Hit_TraceStart :type Vector :default "0, 0, 0") (Hit_TraceEnd :type Vector :default "0, 0, 0"))
  ;[1 1 1] Destroy on stop
  (Utilities|Casting|CastToFortPawn Hit_HitActor
    (:then
      _)
    (:CastFailed
      (Fortnite|ApplyGameplayEffectsToActor Hit_HitActor (Variables|Default|GetHostileImpactGEsToApply) Hit_Location)))
  (CallFunction|HandleDestruction)
  (if (not (ScalableFloat|GetValueasBool (Variables|Default|GetHFProjectileHasAoE)))
    (CallFunction|ApplyEffectstoTarget Hit_HitActor)))

(event Custom|HandleTargeting
  ;[1 1 1] AoE on hit
  (if (ScalableFloat|GetValueasBool (Variables|Default|GetHFProjectileHasAoE))
    (bind (_asynctaskproxy _targetinghandle) (Class|AsyncActionPerformTargeting|PerformTargetingAsyncAction 0 "/HoneyFiber_Incident/Gameplay/Targeting/TP_HoneyFiber_Incident_ImpactAoE.TP_HoneyFiber_Incident_ImpactAoE")
      (:Targeted
        (for _array_element (TargetingSystem|TargetingResults|GetTargetingResultsActors (GameInstanceSubsystems|GetTargetingSubsystem) _targetinghandle)
          (if (ScalableFloat|GetValueasBool (Variables|Default|GetHF_RequireLineOfSightForTargeting))
            (bind _returnvalue (Fortnite|CheckLineOfSightToActorWithChannel (Transformation|GetActorLocation (Variables|Actor|GetInstigator)) _array_element "ECC_Visibility" (Variables|Actor|GetInstigator)))
            (if _returnvalue
              _)
            (else
              _))
          (CallFunction|ApplyEffectstoTarget _array_element))))))

(event Custom|HandleDestruction
  (bind _projectiletrail (Variables|Default|GetProjectileTrail))
  ;[1 1 1] Decouple trail from projectile, play impact VFX and destroy the projectile
  (Niagara|SetNiagaraVariable(Bool) _projectiletrail "Hit" true)
  (Components|Activation|Deactivate _projectiletrail)
  (CallFunction|HandleTargeting)
  (GameplayCue|ExecuteGameplayCueOnActor(Burst) self (Variables|Default|GetGCNeutralImpact) (Ability|GameplayCue|MakeGameplayCueParameters :NormalizedMagnitude 0.0 :RawMagnitude 0.0 :Location (Transformation|GetActorLocation) :Normal "0,0,0" :GameplayEffectLevel 1 :AbilityLevel 1 :bReplicateLocationWhenUsingMinimalRepProxy false))
  (Projectile|SetVelocity (Variables|Movement|GetProjectileMovementComponent) (0.0 0.0 0.0))
  (Collision|SetCollisionEnabled (Variables|Collision|GetCapsuleComponent) "NoCollision")
  (Audio|Components|Audio|FadeOut (Variables|Default|GetSound) 0.5 0.0 "Linear" :!class /Script/Engine.AudioComponent)
  (Utilities|FlowControl|Delay (ScalableFloat|GetValueatLevel (Variables|Default|GetHFProjectileDelayBeforeDestroying)))
  (Actor|DestroyActor))

(fn ApplyEffectsToTarget
  (params (Target :type Actor%20Object%20Reference))
  (bind _instigator (Variables|Actor|GetInstigator))
  (bind _currenttarget (Variables|Default|GetCurrentTarget))
  (bind _returnvalue (Team|GetTeamAffiliation _currenttarget _instigator))
  (Variables|Default|SetCurrentTarget Target)
  (Networking|SwitchHasAuthority
    (:Authority
      (if (not (> (Utilities|Array|FindItem (Variables|Default|GetAlreadyHitActors) (Variables|Default|GetCurrentTarget)) -1))
        (Utilities|Array|AddUnique (Variables|Default|GetAlreadyHitActors) (Variables|Default|GetCurrentTarget))
        (Fortnite|ApplyGameplayEffectsToActor (Variables|Default|GetCurrentTarget) (Utilities|Select :Index _returnvalue :Friendly (Variables|Default|GetFriendlyImpactGEsToApply) :Neutral (Variables|Default|GetHostileImpactGEsToApply) :Hostile (Variables|Default|GetHostileImpactGEsToApply)) (Transformation|GetActorLocation self) false 0 (Ability|GetAbilitySystemComponent (Game|GetInstigator)) (Class|AthenaFunctionLibrary|MakeSimpleFortEffectContext (Game|GetInstigator) (Game|GetInstigator)))
        (GameplayCue|ExecuteGameplayCueOnActor(Burst) (Variables|Default|GetCurrentTarget) (Utilities|Select :Index _returnvalue :Friendly (Variables|Default|GetGCFriendlyImpact) :Hostile (Variables|Default|GetGCHostileImpact)) (Ability|GameplayCue|MakeGameplayCueParameters :NormalizedMagnitude 0.0 :RawMagnitude 0.0 :Location (Transformation|GetActorLocation) :Normal "0,0,0" :GameplayEffectLevel 1 :AbilityLevel 1 :bReplicateLocationWhenUsingMinimalRepProxy false))
        (bind _asfort_player_pawn (Utilities|Casting|CastToFortPlayerPawn (Variables|Default|GetCurrentTarget)))
        (if (and (< (FortPawn|GetHealth _asfort_player_pawn) (FortPawn|GetMaxHealth _asfort_player_pawn)) (Utilities|Enum|Equal(Enum) (Team|GetTeamAffiliation (Variables|Default|GetCurrentTarget) (Variables|Actor|GetInstigator)) "Friendly"))
          (GameplayCue|ExecuteGameplayCueOnActor(Burst) _asfort_player_pawn (Variables|Default|GetGCImpactHealing) (Ability|GameplayCue|MakeGameplayCueParameters 0.0))
          (elif (and (ScalableFloat|GetValueasBool (Variables|Default|GetHFHostileCrumbTrailEnabled)) (Utilities|Enum|Equal(Enum) (Team|GetTeamAffiliation (Variables|Default|GetCurrentTarget) (Variables|Actor|GetInstigator)) "Hostile"))
            ;[1 0.1481 0] If a hostile is hit, create a trail of crumbs
            (GameplayEffects|ApplyGameplayEffectToTarget :self (Ability|GetAbilitySystemComponent (Variables|Default|GetCurrentTarget)) :GameplayEffectClass (Variables|Default|GetGCHostileCrumbTrail) :Target (Ability|GetAbilitySystemComponent (Variables|Default|GetCurrentTarget)) :Level 0.0)))))))

(fn ConstructionScript)
